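# Niagara system for the liquid pour in the Concentrate emote, built from a ribbon stream and mesh drops.
# The ribbon falls under gravity and drag, links particles by normalized age, thins over emitter age and drives texture time from negative age.
# Delayed drops spawn at a falling rate with decaying launch speed, accelerate under ramping gravity and scale by lifetime and speed.
# Both emitters inherit partial source velocity and animate dynamic material parameters.
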NiagaraSystem /BRCosmetics/Effects/Fort_Effects/Emotes/Concentrate/NS_Emote_Concentrate_Pour
dependencies {
 usedDataInterfaces [&NiagaraDataInterfaceCurve, &NiagaraDataInterfaceVectorCurve]
 usedDynamicInputs []
 usedModules [&SystemState]
}
emitters {
 drops {
  emitterUpdateScript [
   &EmitterState(Life Cycle Mode=::NewEnumerator1, Loop Delay=0.6583, Loop Duration=0.3625, UseLoopDelay=true)
   &SpawnRate(
    SpawnRate=&FloatFromCurve(FloatCurve=DataInterface() {propertyValues {Curve {keys [<at 0 1>,<at 1 0>]}}}, CurveIndex=$Emitter.NormalizedLoopAge, Scale Curve=40)
    Use Spawn Probability=false
   )
  ]
  particleSpawnScript [
   &InitializeParticle(
    Lifetime=0.6
    Mesh Scale Mode=::NewEnumerator1
    Mesh Uniform Scale Max=0.5
    Mesh Uniform Scale Min=0.3
    UsePositionOffset=true
    Write Material Random=true
   ) {module=InitializeParticle001}
   &ApplyInitialForces()
   &SphereLocation(
    MaskSpawnGroup=false
    Offset=<0 1 -2>
    Offset Coordinate Space:ENiagaraCoordinateSpace=::World
    Override Seed=false
    Sphere Radius=0.1
    Surface Only=false
    UseNonUniformScale=false
    UseOffset=true
    UseSphereOrientation=false
   )
   &AddVelocity(
    Velocity=<0 20 0>
    Velocity Speed Scale=&FloatFromCurve(
     FloatCurve=DataInterface() {propertyValues {Curve {keys [
         {arriveTangent=-0.7603, interpMode=RCIM_Cubic, leaveTangent=-0.7603, tangentMode=RCTM_User, time=0, value=1}
         {arriveTangent=-3.2581, interpMode=RCIM_Cubic, leaveTangent=-3.2581, tangentMode=RCTM_User, time=0.84, value=0}
        ]}}}
     CurveIndex=$Emitter.Age
     Scale Curve=0.5
    )
   )
   &InheritVelocity(Inherited Velocity Amount Scale=<0.1 0.1 1>, Speed Limit=true, UseSpeedThreshold=false)
  ]
  particleUpdateScript [
   &ParticleState()
   &GravityForce(Gravity=&Multiply_VectorByFloat(
     Vector=<1 1 -400>
     Float=&FloatFromCurve(FloatCurve=DataInterface() {propertyValues {Curve {keys [{arriveTangent=1.9078, interpMode=RCIM_Cubic, leaveTangent=1.9078, tangentMode=RCTM_User, time=0, value=0.1}, <at 1 1>]}}})
    ))
   &SolveForcesAndVelocity(Clamp Velocity=false, Limit Acceleration=false)
   &ScaleMeshSize(Scale Factor=&Multiply_Vector(
     A=<1 1 1>
     B=&VectorFromCurve(VectorCurve=DataInterface() {propertyValues {
        XCurve {keys [
          <at 0 0.5 RCIM_Cubic>
          {arriveTangent=0.5, interpMode=RCIM_Cubic, leaveTangent=-5.3087, tangentMode=RCTM_Break, time=0.3, value=2.4}
          <at 1 0.5>
         ]}
        YCurve {keys [<at 0 1>,<at 1 1>]}
        ZCurve {keys [<at 0 1>,<at 1 1>]}
       }})
    ))
   &DynamicMaterialParameters(
    Index 0 Param 1=0.8
    Index 0 Param 2=&FloatFromCurve(FloatCurve=DataInterface() {propertyValues {Curve {keys [{arriveTangent=-1.8884, interpMode=RCIM_Cubic, leaveTangent=-1.8884, tangentMode=RCTM_User, time=0, value=1}, <at 1 0.26>]}}})
    Index 0 Param 3=1
    Index 0 Param 4=1
    Module.Param0WriteEnabled=true
    Module.Param1WriteEnabled=true
    Module.Param2WriteEnabled=true
    Module.Param3WriteEnabled=true
   )
   &ScaleMeshSizeBySpeed(Max Scale Factor=<1 1 1>, Min Scale Factor:Vector3f=<2 1 1>)
  ]
  rendererClasses [&NiagaraMeshRendererProperties]
  renderers [{rendererClass=&NiagaraMeshRendererProperties, rendererIndex=0}]
 }
 ribbon {
  emitterUpdateScript [
   &EmitterState(Life Cycle Mode=::NewEnumerator1, Loop Delay=0.0388, Loop Duration=0.9, UseLoopDelay=true)
   &SpawnRate(SpawnRate=20, Use Spawn Probability=false)
  ]
  particleSpawnScript [
   &InitializeParticle(Position Offset=<0 -2 0>, Sprite Size Mode=::NewEnumerator3, Uniform Sprite Size=1, UsePositionOffset=true, Write Material Random=true) {module=InitializeParticle001}
   /SetVariables(Particles.RibbonWidth=15) {isSetParameters=true}
   &AddVelocity(Velocity=<0 30 0>)
   &InheritVelocity(Inherited Velocity Amount Scale=<0.1 1 0>, Speed Limit=true, UseSpeedThreshold=false)
   &CameraOffset(Camera Offset Amount=-5) {enabled=false}
  ]
  particleUpdateScript [
   &ParticleState()
   &Drag(Drag=0.9, Use Linear Drag=true, Use Rotational Drag=false)
   &GravityForce(Gravity=<0 0 -400>)
   &SolveForcesAndVelocity(Clamp Velocity=false, Limit Acceleration=false)
   &Material(Base Texture Time:NiagaraFloat=&Multiply_Float(A=$Emitter.Age, B=-1))
   &DynamicMaterialParameters(
    Index 0 Param 1=&MaskFloatBySpawnGroup(Masked Float=0, Passthrough Float=1)
    Index 0 Param 2=&FloatFromCurve(FloatCurve=DataInterface() {propertyValues {Curve {keys [{arriveTangent=-2.1032, interpMode=RCIM_Cubic, leaveTangent=-2.1032, tangentMode=RCTM_User, time=0, value=1}, <at 1 0.26>]}}})
    Module.Param0WriteEnabled=true
    Module.Param1WriteEnabled=true
    Module.Param2WriteEnabled=false
    Module.Param3WriteEnabled=false
   )
   &ScaleRibbonWidth(Ribbon Width Scale=&FloatFromCurve(
     FloatCurve=DataInterface() {propertyValues {Curve {keys [{arriveTangent=-1.8723, interpMode=RCIM_Cubic, leaveTangent=-1.8723, tangentMode=RCTM_User, time=0.1, value=1}, <at 0.9 0>]}}}
     CurveIndex=$Emitter.Age
    ))
   /SetVariables(Particles.RibbonLinkOrder:NiagaraFloat=$Particles.NormalizedAge) {isSetParameters=true}
  ]
  rendererClasses [&NiagaraRibbonRendererProperties]
  renderers [{rendererClass=&NiagaraRibbonRendererProperties}]
 }
}
summary {
 systemName=NS_Emote_Concentrate_Pour
 emitters [
  {emitterName=ribbon, rendererClasses [&NiagaraRibbonRendererProperties]}
  {emitterName=drops, rendererClasses [&NiagaraMeshRendererProperties]}
 ]
 userVariables [User.RibbonIDOffset {type=NiagaraInt32}]
}
