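# Niagara effect for the Concentrate emote that continuously spawns small water splashes.
# CrownSplashMesh emits short-lived, randomly scaled crown meshes around a tiny ring, using curves to pop them up, shrink them and fade alpha.
# A GPU sprite emitter sprays speed-stretched droplets from a small flat disc, slowed by drag and gravity while fading out.
# Spawn-group cutoff inputs on the droplets switch particle scale and sprite size.
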NiagaraSystem /BRCosmetics/Effects/Fort_Effects/Emotes/Concentrate/NS_Emote_Concentrate_Splash
dependencies {
 usedDataInterfaces [&NiagaraDataInterfaceCurve, &NiagaraDataInterfaceColorCurve, &NiagaraDataInterfaceVectorCurve]
 usedDynamicInputs []
 usedModules [&SystemState]
}
emitters {
 CrownSplashMesh {
  emitterUpdateScript [
   &EmitterState()
   &SpawnRate(SpawnRate=40, Use Spawn Probability=false)
  ]
  particleSpawnScript [
   /Niagara/Modules/Spawn/Initialization/InitializeParticle(
    Lifetime=0.3
    Mesh Scale=&Multiply_VectorByFloat(Vector=<1 1 1>, Float=&RandomRangeFloat(Minimum=0.2, Maximum=0.6))
    Module.Write Color=true
    Module.Write Lifetime=true
    Module.Write Mass=true
    Module.Write Material Random=true
    Module.Write Position=true
    Module.Write Scale=true
    Module.Write SpriteRotation=false
    Module.Write SpriteSize=true
    Position=&SimulationPosition()
   )
   &AddVelocityInCone(
    Cone Angle=80
    Cone Axis=<0 0 1>
    Module.Use Velocity Falloff On Cone Axis=false
    Velocity Strength=/Game/Effects/Niagara/DynamicInputs/FN_PickFloatFromSpawnGroupCutoff(
     Over Cutoff Float:NiagaraFloat=0
     Under Cutoff Float:NiagaraFloat=&UniformRangedFloat(Minimum=50, Maximum=250)
     SpawnGroup:NiagaraInt32=$Emitter.SpawnGroup
     Cutoff:NiagaraInt32=2
    )
   ) {enabled=false}
   /SetVariables() {enabled=false, isSetParameters=true}
   /SetVariables() {isSetParameters=true}
   &InitialMeshOrientation(Rotation=&UniformRangedVector(Minimum=<0 0 0>, Maximum=<0 0 1>), Use Rotation Vector=true) {module=InitialMeshOrientation001}
   &ShapeLocation(MaskSpawnGroup=false, Override Seed=false, Ring Radius=1, Shape Primitive=::NewEnumerator4, UseOwnerScale=true)
   &AddVelocity(Velocity=<0 0 5>) {module=AddVelocity001}
  ]
  particleUpdateScript [
   &ParticleState()
   &ScaleColor(
    Scale Alpha=&FloatFromCurve(FloatCurve=DataInterface() {propertyValues {Curve {keys [<at 0.65 1.0113>,<at 1.0011 0.002>]}}})
    ScaleA=true
    ScaleRGB=true
   )
   /SetVariables(Particles.MeshColor:LinearColor=&ScaleLinearColorByCurve(
     InitialValue=<rgba 1 1 1 1>
     ScaleCurve=DataInterface() {propertyValues {
       AlphaCurve {keys [<at 0.2012 1>,<at 0.5654 0>]}
       BlueCurve {keys [<at 0.3226 1>,<at 0.9987 0>]}
       GreenCurve {keys [<at 0.3226 1>,<at 0.9987 0>]}
       RedCurve {keys [<at 0.3226 1>,<at 0.9987 0>]}
      }}
     CurveIndex=$Particles.NormalizedAge
    )) {enabled=false, isSetParameters=true}
   &Drag(Use Linear Drag=true, Use Rotational Drag=false)
   &GravityForce(Gravity=<0 0 -100>)
   &SolveForcesAndVelocity(Clamp Velocity=false, Limit Acceleration=false)
   &ScaleMeshSize(Scale Factor=&VectorFromCurve(VectorCurve=DataInterface() {propertyValues {
       XCurve {keys [{interpMode=RCIM_Cubic, leaveTangent=4.8413, tangentMode=RCTM_Break, time=0, value=0}, <at 0.7127 0.5793>, <at 1 0>]}
       YCurve {keys [{interpMode=RCIM_Cubic, leaveTangent=4.8413, tangentMode=RCTM_Break, time=0, value=0}, <at 0.7127 0.5793>, <at 1 0>]}
       ZCurve {keys [
         {interpMode=RCIM_Cubic, leaveTangent=11.3233, tangentMode=RCTM_Break, time=0, value=0}
         {arriveTangent=-3.7658, interpMode=RCIM_Cubic, leaveTangent=-3.7658, tangentMode=RCTM_User, time=0.0765, value=1.4}
         <at 0.9989 0.1>
        ]}
      }}))
   /Niagara/Modules/Update/Orientation/V2/UpdateMeshOrientation(Rotation Rate=0.2, Rotation Vector:Vector3f=<0 0 1>) {enabled=false}
  ]
  rendererClasses [&NiagaraMeshRendererProperties]
  renderers [{rendererClass=&NiagaraMeshRendererProperties, rendererIndex=0}]
 }
 splash {
  simTarget=GPUComputeSim
  emitterUpdateScript [
   &EmitterState(Life Cycle Mode=::NewEnumerator1, Loop Behavior=::NewEnumerator1, Loop Duration=1, UseLoopDelay=false)
   &SpawnRate(SpawnRate=140, Use Spawn Probability=false) {module=SpawnRate001}
   /SetVariables() {enabled=false, isSetParameters=true}
   &SpawnBurst_Instantaneous(Spawn Count=40, Use Spawn Probability=false, UseLoopCountLimit=false) {enabled=false}
  ]
  particleSpawnScript [
   &InitializeParticle(
    Color=<rgba 1 1 1 3>
    Lifetime=&UniformRangedFloat(Minimum=0.2, Maximum=0.3)
    Mesh Scale=<0.5 0.5 0.5>
    Module.Write Color=true
    Module.Write Lifetime=true
    Module.Write Mass=true
    Module.Write Material Random=true
    Module.Write Position=false
    Module.Write Scale=true
    Module.Write SpriteRotation=false
    Module.Write SpriteSize=true
    Position=&SimulationPosition()
    Sprite Size=&Multiply_Vector2(
     A=<2 3>
     B=&Vector2DFromFloat(Value=&UniformRangedFloat(Minimum=0.1, Maximum=1))
    )
   )
   &AddVelocity(
    Module.Use Falloff Distance=false
    Origin Offset Coordinate Space=::World
    Override Seed=false
    Velocity Mode=::NewEnumerator1
    Velocity Speed=&RandomRangeFloat(Minimum=20, Maximum=40)
   )
   /SetVariables(Particles.Scale=/Game/Effects/Niagara/DynamicInputs/Water/FN_PickVectorFromSpawnGroupCutoff(
     Under Cutoff Vector:Vector3f=<0 0 0>
     Over Cutoff Vector:Vector3f=<1 1 1>
     SpawnGroup:NiagaraInt32=$Emitter.SpawnGroup
     Cutoff:NiagaraInt32=2
    )) {isSetParameters=true}
   &CylinderLocation(
    Cylinder Height=0
    Cylinder Height Midpoint=0.1
    Cylinder Radius=2
    MaskSpawnGroup=false
    Offset=<0 0 1>
    Offset Coordinate Space:ENiagaraCoordinateSpace=::World
    Offset Mode:ENiagara_OffsetMode=::NewEnumerator0
    Override Seed=false
    UseOwnerScale=true
   )
   &AddVelocity(Velocity=<0 0 5>) {module=AddVelocity001}
  ]
  particleUpdateScript [
   &ParticleState()
   &GravityForce(Gravity=<0 0 -50>)
   &ScaleColor(
    Scale Alpha=&FloatFromCurve(FloatCurve=DataInterface() {propertyValues {Curve {keys [<at 0.4124 1.0113>,<at 1.0011 0.002>]}}})
    ScaleA=true
    ScaleRGB=false
   )
   &Drag(Drag=5, Use Linear Drag=true, Use Rotational Drag=false)
   &SolveForcesAndVelocity(Clamp Velocity=false, Limit Acceleration=false)
   &ScaleSpriteSizeBySpeed(Max Scale Factor=<1 5>)
   &ScaleSpriteSize(Scale Factor=&Vector2DFromFloat(Value=&FN_PickFloatFromSpawnGroupCutoff(
      Over Cutoff Float:NiagaraFloat=0
      Under Cutoff Float:NiagaraFloat=&FloatFromCurve(FloatCurve=DataInterface() {propertyValues {Curve {keys [
           {interpMode=RCIM_Cubic, tangentMode=RCTM_Break, time=0.5, value=0.9489}
           {arriveTangent=-3.7049, interpMode=RCIM_Cubic, tangentMode=RCTM_Break, time=1.0011, value=0.002}
          ]}}})
      SpawnGroup:NiagaraInt32=$Emitter.SpawnGroup
      Cutoff:NiagaraInt32=2
     )))
  ]
  rendererClasses [&NiagaraSpriteRendererProperties]
  renderers [{rendererClass=&NiagaraSpriteRendererProperties}]
 }
}
summary {
 systemName=NS_Emote_Concentrate_Splash
 emitters [
  {emitterName=splash, simTarget=GPUComputeSim, rendererClasses [&NiagaraSpriteRendererProperties]}
  {emitterName=CrownSplashMesh, rendererClasses [&NiagaraMeshRendererProperties]}
 ]
 userVariables [
  User.bIsFrontEnd {type=NiagaraBool}
  User.bIsFrontEndPreview {type=NiagaraBool}
  User.RibbonIDOffset {type=NiagaraInt32}
 ]
}
